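# Additive, unlit, two-sided material for the glowing eye effect on the vampire bat glider, usable on skeletal meshes.
# A panned noise texture, faded toward all four UV edges, drives opacity and masks a tinted, boosted emissive color.
# World position offset adds a sine/cosine wave that scrolls over time, is phase-shifted by the second UV channel and grows along U.
# The attributes pass through the MF_TechArtOverrides_accessories function.
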Material /BRCosmetics/Gadgets/Assets/VinderTech_GliderChute/Glider_Vampire/Materials/M_Glider_Bat_EyeGlow
graph {
 properties {
  bUseMaterialAttributes=true
  bUsedWithMeshDeformer=true
  bUsedWithSkeletalMesh=true
  blendMode=BLEND_Additive
  pixelDepthOffsetMode=PDOM_Legacy
  shadingModel=MSM_Unlit
  twoSided=true
  usedShadingModels=Unlit
 }
 nodes [
  ComponentMask(None=TextureCoordinate() {coordinateIndex=1}) {r=true}
  Constant()
  n1 = Multiply(
   A=ScalarParameter() {defaultValue=30.8296, parameterName=Glow}
   B=VectorParameter() {parameterName=Color, defaultValue=<rgba 1 0.474 0.9052 1>}.RGB
  )
  n2 = TextureCoordinate()
  n3 = ComponentMask(None=@n2) {r=true}
  n4 = TextureCoordinate()
  n5 = AppendVector(A=ScalarParameter() {defaultValue=0.5208, parameterName=TilingNoise}, B=ScalarParameter() {defaultValue=0.1076, parameterName=TilingNoise_Y})
  n6 = ScalarParameter() {defaultValue=-0.1616, parameterName=Speed_X}
  n7 = ScalarParameter() {defaultValue=0.1, parameterName=Speed_Y}
  n8 = TextureSample(UVs=Panner(Coordinate=Multiply(A=@n4, B=@n5), Speed=AppendVector(A=@n6, B=@n7))) {samplerType=SAMPLERTYPE_Masks, texture=/Game/Gadgets/Assets/VinderTech_GliderChute/Glider_Vampire/Textures/T_Glider_Bat_Noises}
  n9 = ComponentMask(None=@n2) {g=true}
  n10 = Multiply(A=Multiply(A=Multiply(A=@n3, B=@n8.G), B=OneMinus(None=@n3)), B=Multiply(A=@n9, B=OneMinus(None=@n9)))
  n11 = TextureCoordinate()
  n12 = Time()
  n13 = ScalarParameter() {defaultValue=1.4008, parameterName=WaveSpeed}
  n14 = ComponentMask(None=TextureCoordinate() {coordinateIndex=1}) {g=true}
  n15 = ScalarParameter() {defaultValue=14.4546, parameterName=Timeoffset}
  n16 = Add(A=@n11, B=Add(A=Multiply(A=@n12, B=@n13), B=Multiply(A=@n14, B=@n15)))
  n17 = Frac(None=Multiply(A=ComponentMask(None=@n16) {r=true}, B=ScalarParameter() {defaultValue=0.9577, parameterName=WaveLength}))
  n18 = Sine(None=@n17)
  n19 = AppendVector(A=@n18, B=Cosine(None=@n17))
  n20 = ComponentMask(None=@n11) {r=true}
  n21 = ScalarParameter() {defaultValue=5.9791, parameterName=OffsetDistance}
  n22 = Multiply(
   A=VectorParameter() {parameterName=OffsetDirectionVector, defaultValue=<rgba 0 1 1 1>}.RGB
   B=Multiply(A=AppendVector(A=@n19, B=@n18), B=Multiply(A=@n20, B=@n21))
  )
  MaterialFunctionCall(In=SetMaterialAttributes!(
    MaterialAttributes
    Emissive Color=Multiply(A=@n1, B=@n10)
    Opacity=@n10
    World Position Offset=Transform(None=@n22) {transformSourceType=TRANSFORMSOURCE_Local}
   ) {attributeSetTypes [#1, #2, #3]}) {materialFunction=&MF_TechArtOverrides_accessories}
 ]
}
