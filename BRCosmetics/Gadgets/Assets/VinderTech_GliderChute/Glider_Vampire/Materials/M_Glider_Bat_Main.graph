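# Master material for the Vampire bat glider, with an IsTranslucentLayer switch for a smoky translucent layer.
# Panning, self-distorting noise modulates color and specular, while an emissive mask adds glow.
# The translucent path fakes sun diffuse and specular in emissive and builds noisy tail opacity with HeightLerp.
# Vertex-color-masked sine waves flap the wings through WPO, with optional camera push and Blueprint-named velocity offset.
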Material /BRCosmetics/Gadgets/Assets/VinderTech_GliderChute/Glider_Vampire/Materials/M_Glider_Bat_Main
graph {
 properties {
  bEnableResponsiveAA=true
  bUseMaterialAttributes=true
  bUsedWithMeshDeformer=true
  bUsedWithSkeletalMesh=true
  pixelDepthOffsetMode=PDOM_Legacy
  usedShadingModels=Default Lit
 }
 nodes [
  n1 = CameraPositionWS()
  n2 = WorldPosition()
  n3 = Multiply(A=Normalize(VectorInput=Subtract(A=@n1, B=@n2.XYZ)), B=ScalarParameter() {defaultValue=1, parameterName=PushOffset})
  n4 = PreSkinnedPosition()
  Add(A=@n3, B=@n4)
  n5 = TextureSample() {texture=/BRCosmetics/Gadgets/Assets/VinderTech_GliderChute/Glider_Vampire/Textures/T_Glider_Bat_Main_D}
  n6 = Multiply(A=@n5.RGB, B=ScalarParameter() {defaultValue=0.1952, parameterName=DiffuseBrightness})
  n7 = TextureCoordinate() {coordinateIndex=1}
  n8 = ScalarParameter() {defaultValue=1.7998, parameterName=NoiseTiling}
  n9 = Time()
  n10 = ScalarParameter() {defaultValue=26.8502, parameterName=PanSpeed}
  n11 = TextureSample(UVs=Panner(Coordinate=Multiply(A=@n7, B=@n8), Time=Multiply(A=@n9, B=@n10)) {bFractionalPart=true, speedY=0.01}) {samplerType=SAMPLERTYPE_Masks, texture=/Game/Gadgets/Assets/VinderTech_GliderChute/Glider_Vampire/Textures/T_Glider_Bat_Noises}
  n12 = Add(
   A=TextureCoordinate() {coordinateIndex=1}
   B=Multiply(A=@n11.B, B=ScalarParameter() {defaultValue=0.0686, parameterName=Distortion_Diffuse_1})
  )
  n13 = Clamp(None=LinearInterpolate(
    A=ScalarParameter() {defaultValue=-0.044, parameterName=DiffuseNoiseContrast_1}
    B=ScalarParameter() {defaultValue=0.9232, parameterName=DiffuseNoiseContrast_2}
    Alpha=TextureSample(UVs=@n12) {samplerType=SAMPLERTYPE_Masks, texture=&T_Glider_Bat_Noises}.G
   ))
  n14 = TextureSample() {samplerType=SAMPLERTYPE_Masks, texture=/BRCosmetics/Gadgets/Assets/VinderTech_GliderChute/Glider_Vampire/Textures/T_Glider_Bat_Main_E}
  n15 = LinearInterpolate(A=Constant() {r=1}, B=@n13, Alpha=@n14.R)
  n16 = Multiply(A=@n6, B=@n15)
  n17 = TextureSample() {samplerType=SAMPLERTYPE_Masks, texture=/BRCosmetics/Gadgets/Assets/VinderTech_GliderChute/Glider_Vampire/Textures/T_Glider_Bat_Main_M}
  n18 = ScalarParameter() {defaultValue=0.2048, parameterName=Specular}
  n19 = LinearInterpolate(A=ScalarParameter() {parameterName=Specular_BlackNoise}, B=@n18, Alpha=@n13)
  n20 = Add(A=TextureCoordinate(), B=Multiply(A=@n11.B, B=ScalarParameter() {defaultValue=0.0286, parameterName=Distortion_Diffuse}))
  n21 = ComponentMask(None=CollectionParameter() {collection=&FortniteMaterialParameters, parameterName=SunLightColor}) {b=true, g=true, r=true}
  n22 = ScalarParameter() {defaultValue=0.5, parameterName=DesatReallight}
  n23 = CollectionParameter() {
   collection=&FortniteMaterialParameters
   parameterName=SunlightDirectionalVector
  }
  n24 = ComponentMask(None=@n23) {b=true, g=true, r=true}
  n25 = TextureCoordinate()
  n26 = TextureCoordinate() {coordinateIndex=1}
  n27 = ScalarParameter() {defaultValue=0.9043, parameterName=NoiseTiling_2}
  n28 = Time()
  n29 = ScalarParameter() {defaultValue=14.6561, parameterName=PanSpeed_2}
  n30 = TextureSample(UVs=Panner(Coordinate=Multiply(A=@n26, B=@n27), Time=Multiply(A=@n28, B=@n29)) {bFractionalPart=true, speedX=0.001, speedY=0.01}) {samplerType=SAMPLERTYPE_Masks, texture=&T_Glider_Bat_Noises}
  n31 = Multiply(A=@n30.B, B=ScalarParameter() {defaultValue=0.4667, parameterName=Distortion_2})
  n32 = MaterialFunctionCall(
   Normal=TextureSample(UVs=Add(A=@n25, B=@n31)) {samplerType=SAMPLERTYPE_Normal, texture=/Game/Gadgets/Assets/VinderTech_GliderChute/Glider_Vampire/Textures/T_Glider_Bat_Noises_N}.RGB
   Flatness=ScalarParameter() {defaultValue=1.8684, parameterName=NormalFlatten}
  ) {materialFunction=&FlattenNormal}
  n33 = Constant3Vector() {constant=<rgba 1 1 0 0>}
  n34 = TextureSample() {samplerType=SAMPLERTYPE_Normal, texture=/BRCosmetics/Gadgets/Assets/VinderTech_GliderChute/Glider_Vampire/Textures/T_Glider_Bat_Main_N}
  n35 = Transform(None=Add(A=Multiply(A=@n32.Result, B=@n33), B=@n34.RGB))
  n36 = DotProduct(A=@n24, B=@n35)
  n37 = LinearInterpolate(
   A=VectorParameter() {parameterName=FakeFresnell, defaultValue=<rgba 0.0002 0.003 0.01 1>}.RGB
   B=Multiply(A=VectorParameter() {parameterName=FakeFresnell_2, defaultValue=<rgba 0.005 0.0039 0.0035 1>}.RGB, B=Desaturation(None=@n21, Fraction=@n22))
   Alpha=Power(Base=Clamp(None=@n36), Exp=ScalarParameter() {defaultValue=1, parameterName=FakelLightPower})
  )
  n38 = CameraVectorWS()
  n39 = DotProduct(A=@n35, B=Normalize(VectorInput=Add(A=@n24, B=@n38)))
  n40 = Multiply(
   A=VectorParameter() {parameterName=SpecularColor, defaultValue=<rgba 0.01 0.0057 0.0056 1>}.RGB
   B=Power(Base=Saturate(None=@n39), Exp=ScalarParameter() {defaultValue=15.9283, parameterName=FakelSpecPower_1})
  )
  n41 = Add(
   A=Multiply(
    A=ScalarParameter() {defaultValue=0.3, parameterName=EmissiveBaseColorBirghtness}
    B=TextureSample(UVs=@n20) {texture=&T_Glider_Bat_Main_D}.RGB
   )
   B=Multiply(A=ScalarParameter() {defaultValue=0.58, parameterName=BrightnessFakeLight}, B=Add(A=@n37, B=@n40))
  )
  n42 = Multiply(A=@n15, B=ScalarParameter() {defaultValue=0.219, parameterName=EmissiveBaseColorBirghtness_Opaque})
  n43 = OneMinus(None=@n17.G)
  n44 = VectorParameter() {parameterName=EmissiveColor, defaultValue=<rgba 1 0.3503 0.8798 1>}
  n45 = Add(
   A=Multiply(A=@n5.RGB, B=Multiply(A=@n42, B=@n43))
   B=Multiply(A=ScalarParameter() {defaultValue=46.6372, parameterName=Glow}, B=Multiply(A=@n44.RGB, B=@n14.G))
  )
  n46 = VertexColor() {desc=mask flappy parts of the wings}
  n47 = SphereMask(
   A=TextureCoordinate() {coordinateIndex=1}
   B=AppendVector(A=ScalarParameter() {defaultValue=0.5, parameterName=MaskU}, B=ScalarParameter() {defaultValue=0.5, parameterName=MaskV})
   Radius=ScalarParameter() {defaultValue=0.5, parameterName=MaskRadius}
   Hardness=ScalarParameter() {parameterName=MaskHardness}
  )
  n48 = LinearInterpolate(
   A=ScalarParameter() {defaultValue=5.7451, parameterName=WaveAmplitude}
   B=ScalarParameter() {defaultValue=5.7451, parameterName=WaveAmplitudeTail}
   Alpha=@n47
  )
  n49 = StaticSwitchParameter(True=@n4, False=@n4) {parameterName=IsTranslucentLayer}
  n50 = Multiply(A=VectorParameter() {parameterName=DirectionVector, defaultValue=<rgba 0 1 0 1>}.RGB, B=Time())
  n51 = Multiply(A=@n50, B=ScalarParameter() {defaultValue=119, parameterName=WaveSpeed})
  n52 = Divide(A=ComponentMask(None=Add(A=@n49, B=@n51)) {g=true, r=true}, B=ScalarParameter() {defaultValue=67.0053, parameterName=WaveLength})
  n53 = Multiply(A=@n50, B=ScalarParameter() {defaultValue=79, parameterName=WaveSpeed_1})
  n54 = Divide(A=ComponentMask(None=Add(A=@n49, B=@n53)) {g=true, r=true}, B=ScalarParameter() {defaultValue=79.9913, parameterName=WaveLength_1})
  n55 = Add(A=Sine(None=ComponentMask(None=@n52) {r=true}), B=Sine(None=ComponentMask(None=@n54) {g=true}))
  n56 = LinearInterpolate(A=@n55, B=Multiply(A=@n55, B=VectorParameter() {parameterName=WavelengthInside, defaultValue=<rgba 1 1 1 0>}.RGB), Alpha=@n47)
  n57 = Multiply(
   A=Multiply(A=Multiply(A=@n48, B=@n56), B=VertexColor() {desc=mask flappy parts of the wings}.R)
   B=Transform(None=VectorParameter() {parameterName=OffsetDirection, defaultValue=<rgba 0 0 1 1>}.RGB) {transformSourceType=TRANSFORMSOURCE_Local}
  )
  n58 = StaticSwitchParameter(True=Add(A=@n3, B=@n57), False=@n57) {parameterName=IsTranslucentLayer}
  n59 = Add(A=Multiply(A=Multiply(A=@n46.G, B=@n46.G), B=VectorParameter() {parameterName=BP_VelocityOffset, defaultValue=<rgba 0 0 0 1>}.RGB), B=@n58)
  n60 = ScalarParameter() {defaultValue=0.035, parameterName=Distortion}
  n61 = TextureSample(UVs=Add(A=TextureCoordinate(), B=Multiply(A=@n11.B, B=@n60))) {samplerType=SAMPLERTYPE_Masks, texture=&T_Glider_Bat_Main_M}
  n62 = VertexColor()
  n63 = OneMinus(None=@n62.B)
  n64 = ScalarParameter() {defaultValue=0.8667, parameterName=TailSmokeOpacity_3}
  n65 = ScalarParameter() {defaultValue=0.086, parameterName=TailSmokeOpacity}
  n66 = ScalarParameter() {defaultValue=-1.8232, parameterName=TailSmokeOpacity_1}
  n67 = ScalarParameter() {defaultValue=19.0458, parameterName=TailSmokeOpacity_2}
  n68 = OneMinus(None=@n62.G)
  n69 = Power(Base=@n68, Exp=ScalarParameter() {defaultValue=23.7112, parameterName=Transition_1})
  n70 = ScalarParameter() {defaultValue=0.3533, parameterName=Transition_2}
  n71 = TextureCoordinate() {coordinateIndex=1}
  n72 = ScalarParameter() {defaultValue=0.4448, parameterName=NoiseTiling_1}
  n73 = Time()
  n74 = ScalarParameter() {defaultValue=1.4937, parameterName=PanSpeed_1}
  n75 = Add(
   A=Panner(Coordinate=Multiply(A=@n71, B=@n72), Time=Multiply(A=@n73, B=@n74)) {bFractionalPart=true, speedX=0.01, speedY=0.1}
   B=Multiply(A=@n30.B, B=ScalarParameter() {defaultValue=-0.08, parameterName=Distortion_1})
  )
  n76 = MaterialFunctionCall(
   Transition Phase=Multiply(A=@n68, B=ScalarParameter() {defaultValue=0.5314, parameterName=Transition})
   Height Texture=Add(
    A=Multiply(A=@n69, B=@n70)
    B=TextureSample(UVs=@n75) {samplerType=SAMPLERTYPE_Masks, texture=&T_Glider_Bat_Noises}.R
   )
   Contrast=@n65
  ) {materialFunction=&HeightLerp}
  n77 = Clamp(None=Multiply(A=@n65, B=LinearInterpolate(A=@n66, B=@n67, Alpha=@n76.Lerp Alpha No Contrast)))
  n78 = Clamp(None=Add(A=Multiply(A=@n61.R, B=@n63), B=Multiply(A=@n64, B=@n77)))
  n79 = SetMaterialAttributes!(
   MaterialAttributes
   Base Color=NamedRerouteDeclaration(None=LinearInterpolate(A=@n16, B=@n6, Alpha=@n17.G)) {name=baseColor, nodeColor=<rgba 0 0.2266 1 1>}
   Metallic=NamedRerouteDeclaration(None=@n17.G) {name=metallic, nodeColor=<rgba 0.9688 1 0 1>}
   Specular=NamedRerouteDeclaration(None=LinearInterpolate(A=@n18, B=@n19, Alpha=@n14.R)) {name=specular, nodeColor=<rgba 0.125 0 1 1>}
   Roughness=NamedRerouteDeclaration(None=@n17.B) {name=roughness, nodeColor=<rgba 0.4766 0 1 1>}
   Emissive Color=NamedRerouteDeclaration(None=StaticSwitchParameter(True=@n41, False=@n45) {parameterName=IsTranslucentLayer}) {name=emissive, nodeColor=<rgba 1 0.75 0 1>}
   Normal=NamedRerouteDeclaration(None=@n34.RGB) {name=normal, nodeColor=<rgba 0.2422 0 1 1>}
   World Position Offset=NamedRerouteDeclaration(None=StaticSwitchParameter(True=@n59, False=@n58) {parameterName=EnableVelocityOffset}) {name=wpo, nodeColor=<rgba 0.7812 1 0 1>}
   Opacity=StaticSwitchParameter(True=Multiply(A=@n14.R, B=@n78), False=ScalarParameter() {parameterName=SSSDistance}) {parameterName=IsTranslucentLayer}
  ) {attributeSetTypes [
    #1
    #2
    #3
    #4
    #5
    #6
    #7
    #8
   ]}
  MaterialFunctionCall(In=@n79) {materialFunction=&MF_TechArtOverrides_accessories}
  MaterialFunctionCall(BaseNormal=@n34.RGB, AdditionalNormal=@n32.Result) {materialFunction=&BlendAngleCorrectedNormals}
  Multiply(A=@n16, B=Multiply(A=ScalarParameter() {defaultValue=0.5, parameterName=SSSBrightness}, B=OneMinus(None=@n17.G)))
  Saturate(None=Abs(None=@n36))
 ]
}
