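# Translucent, two-sided unlit material for the trim effect on the vampire bat glider, built for skeletal meshes.
# A panning gradient, distorted by panning noise, blends two color parameters and is multiplied by a UV1 mask.
# The dimmed diffuse texture on UV2 feeds an emissive adjustment function.
# Two sine-wave displacement chains use pre-skinned position and time, one faded along UV height, the other masked by red vertex color for flapping wing parts.
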Material /BRCosmetics/Gadgets/Assets/VinderTech_GliderChute/Glider_Vampire/Materials/M_Glider_Bat_TrimEffect
graph {
 properties {
  bEnableResponsiveAA=true
  bUsedWithMeshDeformer=true
  bUsedWithSkeletalMesh=true
  blendMode=BLEND_Translucent
  pixelDepthOffsetMode=PDOM_Legacy
  shadingModel=MSM_Unlit
  twoSided=true
  usedShadingModels=Unlit
 }
 nodes [
  Add()
  n1 = PreSkinnedPosition()
  n2 = VectorParameter() {parameterName=DirectionVector, defaultValue=<rgba 0 1 0 1>}
  n3 = Multiply(A=Multiply(A=@n2.RGB, B=Time()), B=ScalarParameter() {defaultValue=119.9166, parameterName=WaveSpeed})
  n4 = Divide(A=ComponentMask(None=Add(A=@n1, B=@n3)) {g=true, r=true}, B=ScalarParameter() {defaultValue=66.1808, parameterName=WaveLength})
  n5 = ComponentMask(None=@n4) {r=true}
  n6 = ComponentMask(None=@n4) {g=true}
  n7 = Multiply(A=ScalarParameter() {defaultValue=3.0462, parameterName=WaveAmplitude}, B=Add(A=Sine(None=@n5), B=Sine(None=@n6)))
  n8 = OneMinus(None=ComponentMask(None=TextureCoordinate()) {g=true})
  n9 = VectorParameter() {parameterName=OffsetDirection, defaultValue=<rgba 0 0.6122 1.8982 1>}
  Add(A=Multiply(A=Multiply(A=@n7, B=@n8), B=Transform(None=@n9.RGB) {transformSourceType=TRANSFORMSOURCE_Local}))
  n10 = Multiply(A=TextureCoordinate(), B=AppendVector(A=ScalarParameter() {defaultValue=0.2119, parameterName=Tilingx}, B=Constant() {r=1}))
  n11 = AppendVector(A=ScalarParameter() {defaultValue=1.0186, parameterName=SpeedGradient}, B=Constant())
  n12 = TextureCoordinate()
  n13 = ScalarParameter() {defaultValue=0.0844, parameterName=NoiseTiling}
  n14 = Time()
  n15 = ScalarParameter() {defaultValue=26.8502, parameterName=PanSpeed}
  n16 = TextureSample(UVs=Panner(Coordinate=Multiply(A=@n12, B=@n13), Time=Multiply(A=@n14, B=@n15)) {bFractionalPart=true, speedY=0.01}) {samplerType=SAMPLERTYPE_Masks, texture=/Game/Gadgets/Assets/VinderTech_GliderChute/Glider_Vampire/Textures/T_Glider_Bat_Noises}
  n17 = LinearInterpolate(
   A=ScalarParameter() {defaultValue=0.663, parameterName=Distortion_1}
   B=ScalarParameter() {parameterName=Distortion_2}
   Alpha=ComponentMask(None=TextureCoordinate()) {g=true}
  )
  n18 = TextureSample(UVs=Add(A=Panner(Coordinate=@n10, Speed=@n11) {bFractionalPart=true}, B=Multiply(A=@n16.B, B=@n17))) {texture=/BRCosmetics/Gadgets/Assets/VinderTech_GliderChute/Glider_Vampire/Textures/T_Glider_Bat_Gradient}
  LinearInterpolate(
   A=VectorParameter() {parameterName=Color, defaultValue=<rgba 0.01 0.0056 0.0056 1>}.RGB
   B=VectorParameter() {parameterName=Color_1, defaultValue=<rgba 0.01 0.0056 0.0049 1>}.RGB
   Alpha=@n18.R
  )
  n19 = TextureCoordinate() {coordinateIndex=2}
  MaterialFunctionCall(Emissive=Multiply(
    A=TextureSample(UVs=@n19) {texture=/BRCosmetics/Gadgets/Assets/VinderTech_GliderChute/Glider_Vampire/Textures/T_Glider_Bat_Main_D}.RGB
    B=ScalarParameter() {defaultValue=0.2714, parameterName=DiffuseBrightness}
   )) {materialFunction=&MF_FX_EmissiveAdjustments}
  Multiply(A=@n18.R, B=ComponentMask(None=TextureCoordinate() {coordinateIndex=1}) {r=true}) {constA=1}
  n20 = ScalarParameter() {defaultValue=5.7451, parameterName=WaveAmplitude_1}
  n21 = PreSkinnedPosition()
  n22 = VectorParameter() {parameterName=DirectionVector_1, defaultValue=<rgba 0 1 0 1>}
  n23 = Multiply(A=@n22.RGB, B=Time())
  n24 = Multiply(A=@n23, B=ScalarParameter() {defaultValue=119.9166, parameterName=WaveSpeed_1})
  n25 = Divide(A=ComponentMask(None=Add(A=@n21, B=@n24)) {g=true, r=true}, B=ScalarParameter() {defaultValue=67.0053, parameterName=WaveLength_1})
  n26 = Multiply(A=@n23, B=ScalarParameter() {defaultValue=79.4645, parameterName=WaveSpeed_2})
  n27 = Divide(A=ComponentMask(None=Add(A=@n21, B=@n26)) {g=true, r=true}, B=ScalarParameter() {defaultValue=79.9913, parameterName=WaveLength_2})
  n28 = Add(A=Sine(None=ComponentMask(None=@n25) {r=true}), B=Sine(None=ComponentMask(None=@n27) {g=true}))
  Multiply(
   A=Multiply(A=Multiply(A=@n20, B=@n28), B=VertexColor() {desc=mask flappy parts of the wings}.R)
   B=Transform(None=VectorParameter() {parameterName=OffsetDirection_1, defaultValue=<rgba 0 0 1 1>}.RGB) {transformSourceType=TRANSFORMSOURCE_Local}
  )
  TextureSample() {texture=&T_Glider_Bat_Main_D}
  Transform(None=@n2.RGB) {transformSourceType=TRANSFORMSOURCE_Local}
  Transform(None=@n22.RGB) {transformSourceType=TRANSFORMSOURCE_Local}
  WorldPosition()
  WorldPosition()
 ]
}
